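# Physics asset for the Sacred Cuddle backpack cosmetic mesh (F_MED_SacredCuddle_Pack).
# It holds a single capsule body on the pack bone, rotated to lie roughly sideways and fit the bag's shape.
# It has no constraints, so the pack simulates or collides as one rigid piece.
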PhysicsAsset /BRCosmetics/Accessories/FORT_Backpacks/Backpack_F_MED_SacredCuddle/Meshes/F_MED_SacredCuddle_Pack_PhysicsAsset
bodies {pack [{shapeType=Capsule, length=21.7242, radius=11.6995, center=<-0.0025 -2.2842 -0.4373>, rotation=<rot -2.7814 -90 -0.0842>}]}
constraints []
