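# Animation Blueprint for the SacredCuddle backpack, adjusting the pack to the wearer's body and pose.
# MainPose zero-scales the hilts, offsets the pack by body mesh, and tilts it using head-to-spine yaw read from the character mesh.
# During skydive floating it blends in additive top_rotate adjustments, partly scaled by spine bend.
# Weapon, skydive and tactical-sprint events track whether a SacredCuddle-tagged pickaxe is equipped.
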AnimBlueprint /BRCosmetics/Accessories/FORT_Backpacks/Backpack_F_MED_SacredCuddle/Meshes/F_MED_SacredCuddle_Pack_Skeleton_AnimBlueprint
target_skeleton=&/F_MED_SacredCuddle_Pack_Skeleton
graphs [
 AnimGraph {nodes [
   n1 = "Transform (Modify) Bone - Bone: top_rotate"(ComponentPose=Local To Component(LocalPose=Use cached pose 'MainPose'()), Alpha=1, Translation="0,3,-10", Rotation="0,0,23", Scale="1,1,1") {
    rotationMode=BMM_Additive
    translationMode=BMM_Additive
    alphaBoolBlend {blendOption=Linear}
    boneToModify {boneName=top_rotate}
    rotation {roll=23}
    translation {y=3, z=-10}
   }
   n2 = "Transform (Modify) Bone - Bone: top_rotate"(
    ComponentPose=@n1
    Alpha=Get Spine_Rotate_Alpha_par_1() -> (Spine_Rotate_Alpha_par_1:double)
    Translation="0,0,0"
    Rotation="0,0,12"
    Scale="1,1,1"
   ) {
    rotationMode=BMM_Additive
    translationMode=BMM_Additive
    alphaBoolBlend {blendOption=Linear}
    boneToModify {boneName=top_rotate}
    rotation {roll=12}
   }
   n3 = "Transform (Modify) Bone - Bone: top_rotate"(
    ComponentPose=@n2
    Alpha=Get Spine_Rotate_Alpha_part_2() -> (Spine_Rotate_Alpha_part_2:double)
    Translation="0,-2,0"
    Rotation="10,-20,40"
    Scale="1,1,1"
   ) {
    rotationMode=BMM_Additive
    translationMode=BMM_Additive
    alphaBoolBlend {blendOption=Linear}
    boneToModify {boneName=top_rotate}
    rotation=<rot 10 -20 40>
    translation {y=-2}
   }
   Output Pose(Result=Blend Poses by bool(
     BlendPose_0=Component To Local(ComponentPose=@n3)
     BlendPose_1=Use cached pose 'MainPose'()
     BlendTime_0=1.2
     BlendTime_1=0.5
     bActiveValue=Get bIsSkydiveFloating()
    ) {blendType=Linear})
   n4 = "Transform (Modify) Bone - Bone: hilts"!(ComponentPose=Local To Component(LocalPose=Input Pose()), bAlphaBoolEnabled:bool=True, Translation="0,0,0", Rotation="0,0,0", Scale="0,0,0") -> (Pose) {alphaInputType=Bool, scaleMode=BMM_Replace, alphaBoolBlend {blendOption=Linear}, boneToModify {boneName=hilts}, scale=<0 0 0>}
   n5 = "Transform (Modify) Bone - Bone: pack"(
    ComponentPose="Transform (Modify) Bone - Bone: pack"(ComponentPose=@n4, Translation=Get begin_offset() -> (begin_offset:Vector), Rotation="0,0,0") {translationMode=BMM_Additive, alphaBoolBlend {blendOption=Linear}, boneToModify {boneName=pack}}
    Alpha=Get Head_Rotate_Alpha() -> (Head_Rotate_Alpha:double)
    Translation="0,-1,0"
    Rotation="0,0,-10"
   ) {
    rotationMode=BMM_Additive
    translationMode=BMM_Additive
    alphaBoolBlend {blendOption=Linear}
    boneToModify {boneName=pack}
    rotation {roll=-10}
    translation {y=-1}
   }
   Save cached pose 'MainPose'(Pose=Component To Local(ComponentPose=@n5)) {cacheName=MainPose}
   Get bIsSkyDiving()
  ]}
 EventGraph {nodes [
   n1 = Event OnBeginTacticalSprinting() {is=title K2Node_Event}
   Set isPickaxeEquiped(execute=[@n1.then, @n6.then, @n8.else, @n9.then, @n4.else], isPickaxeEquiped:bool=false) -> (Output_Get:bool)
   n2 = Event OnEndTacticalSprinting() {is=title K2Node_Event}
   n3 = PickaxeItemDefinitionHasAnyMetaTag(Tags=Make Array("[0]"=(TagName="Cosmetics.MetaTag.SacredCuddle")), bRequiresEquipped=true)
   n4 = Branch(execute=[@n2.then, @n5.then], Condition=@n3)
   Set isPickaxeEquiped(execute=[@n9.else, @n4.then], isPickaxeEquiped=true)
   n5 = Event OnEndSkydiving()
   n6 = Event OnBeginSkydiving()
   n7 = Event OnWeaponChanged()
   n8 = Branch(execute=[@n7.then, @n12.then_0], Condition=@n3)
   > n9 = Branch(Condition=Get bIsSkyDiving())
   Event BlueprintUpdateAnimation()
   n10 = Get Character_SkeletalMesh() -> (Character_SkeletalMesh:Skeletal Mesh Component Object Reference)
   > Set Head_Rotate_Alpha(Head_Rotate_Alpha:double=MapRangeClamped(
     Value=GetDeltaTransformFromRefPose!(self=@n10, BoneName=Spine_05, BaseName=head) -> (
      ReturnValue_Location:Vector
      ReturnValue_Rotation_Roll:float
      ReturnValue_Rotation_Pitch:float
      ReturnValue_Rotation_Yaw:float
      ReturnValue_Scale:Vector
     ).ReturnValue_Rotation_Yaw
     InRangeA=5
     InRangeB=55
    )) -> (Output_Get:double)
   > Set Spine_Rotate_Alpha_par_1(Spine_Rotate_Alpha_par_1:double=MapRangeClamped(Value=GetDeltaTransformFromRefPose(self=@n10, BoneName=Spine_05, BaseName=Spine_01).ReturnValue_Rotation_Yaw, InRangeA=-10, InRangeB=-25)) -> (Output_Get:double)
   > Set Spine_Rotate_Alpha_part_2(Spine_Rotate_Alpha_part_2:double=MapRangeClamped(Value=GetDeltaTransformFromRefPose(self=@n10, BoneName=Spine_05, BaseName=Spine_01).ReturnValue_Rotation_Yaw, InRangeA=-25, InRangeB=-58)) -> (Output_Get:double)
   n11 = Event OnCharacterCustomizationLoaded()
   > Set Character_SkeletalMesh(Character_SkeletalMesh:Skeletal Mesh Component Object Reference=GetOwnerSkeletalMeshComponent()) -> (Output_Get:Skeletal Mesh Component Object Reference)
   > n12 = Sequence()
   n13 = Get Array Item with EFort Custom Part Type(Array:Array of Custom Character Part Object References=@n11.CustomParts) -> (Output:Custom Character Part Object Reference)
   n14 = Branch(
    execute=@n12.then_1
    Condition="Equal (Object)"(
     A=GetSkeletalMesh!(self:Custom Character Part Object Reference=@n13) -> (ReturnValue:Skeletal Mesh Object Reference)
     B=/DefaultCosmeticsAthena/Characters/Meshes/F_MED_Prime
    )
   )
   > Set begin_offset(begin_offset:Vector="0,-0.6,0") -> (Output_Get:Vector)
   Set begin_offset(execute=@n14.else, begin_offset="0,-1.2,0")
  ]}
]
variables [
 begin_offset {type=Vector}
 Character_SkeletalMesh {type=SkeletalMeshComponent}
 Head_Rotate_Alpha {type=number}
 isPickaxeEquiped {type=boolean}
 isSkydive {type=boolean}
 Spine_Rotate_Alpha_par_1 {type=number}
 Spine_Rotate_Alpha_part_2 {type=number}
]
